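# Animation Blueprint for the Vacation Keytar back bling, adjusting how the keytar sits on the character.
# Two chained Transform Bone nodes move the "offset" bone: one replaces its transform while skydiving, the other nudges it for female characters in the front end.
# Skydiving events and a gender check on customization load set those flags.
# BackpackKeytarOffset results are stored at initialization, but the AnimGraph uses hardcoded values instead.
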AnimBlueprint /SparksCosmetics/Items/Cosmetics/Backpacks/Vacation_Keytar/Vacation_Keytar_Backpack_AnimBP
target_skeleton=/SparksCosmetics/Instrument/Keytar/Vacation/Meshes/Vacation_Keytar_Skeleton
graphs [
 AnimGraph {nodes [
   n1 = "Transform (Modify) Bone - Bone: offset"!(
    ComponentPose=Local To Component(LocalPose=Input Pose())
    bAlphaBoolEnabled:bool=Get IsSkydiving() -> (IsSkydiving:bool)
    Translation="2,5,4"
    Rotation="0,90,0"
    Scale="0.95,0.95,0.95"
   ) -> (Pose) {
    alphaInputType=Bool
    rotationMode=BMM_Replace
    scaleMode=BMM_Replace
    translationMode=BMM_Replace
    alphaBoolBlend {blendOption=Linear}
    boneToModify {boneName=offset}
    rotation {yaw=90}
    scale=<0.95 0.95 0.95>
    translation=<2 5 4>
   }
   n2 = "Transform (Modify) Bone - Bone: offset"(
    ComponentPose=@n1
    bAlphaBoolEnabled=AND Boolean(A=Get isF_Med() -> (isF_Med:bool), B=Get bIsInFrontEnd())
    Translation="0,3,0"
    Rotation="-4,0,-8"
    Scale="0.98,0.98,0.98"
   ) {
    alphaInputType=Bool
    rotationMode=BMM_Additive
    scaleMode=BMM_Replace
    translationMode=BMM_Additive
    alphaBoolBlend {blendOption=Linear}
    boneToModify {boneName=offset}
    rotation {pitch=-4, roll=-8}
    scale=<0.98 0.98 0.98>
    translation {y=3}
   }
   Output Pose(Result=Component To Local(ComponentPose=@n2))
  ]}
 EventGraph {nodes [
   Event BlueprintInitializeAnimation()
   > n1 = BackpackKeytarOffset!(execute) -> (then, Output_Location:Vector, Output_Rotation:Rotator, Output_Scale:Vector)
   > Set BackpackOffsetLocation(BackpackOffsetLocation:Vector=@n1) -> (Output_Get:Vector)
   > Set BackpackOffsetRotation(BackpackOffsetRotation:Rotator=@n1.Output_Rotation) -> (Output_Get:Rotator)
   > Set BackpackOffsetScale(BackpackOffsetScale:Vector=@n1.Output_Scale) -> (Output_Get:Vector)
   n2 = Event OnCharacterCustomizationLoaded()
   > Set Backpack(Backpack:Skeletal Mesh Component Object Reference=@n2.CustomBaseMesh) -> (Output_Get:Skeletal Mesh Component Object Reference)
   > Set Body(Body:Skeletal Mesh Component Budgeted Object Reference=Get Array Item with EFort Custom Part Type(Array=@n2.CustomMeshParts)) -> (Output_Get:Skeletal Mesh Component Budgeted Object Reference)
   n3 = "Get (a copy)"(Array:Array of Custom Character Part Object References=@n2.CustomParts, Dimension 1=1) -> (Output:Custom Character Part Object Reference)
   > Set isF_Med(isF_Med:bool="Equal (Enum)"(
     A:EFortCustomGender Enum=Get GenderPermitted!(self:Custom Character Part Object Reference=@n3) -> (GenderPermitted:EFortCustomGender Enum=Invalid)
     B:EFortCustomGender Enum=Female
    )) -> (Output_Get:bool)
   Event OnEndSkydiving()
   > Set IsSkydiving(IsSkydiving:bool=false) -> (Output_Get:bool)
   Event OnBeginSkydiving()
   > Set IsSkydiving(IsSkydiving=true)
   TryGetPawnOwner()
   "Equal (Enum)"(
    A:EFortCustomBodyType Enum=Get BodyTypesPermitted!(self:Custom Character Part Object Reference=@n3) -> (BodyTypesPermitted:EFortCustomBodyType Enum=NONE)
    B:EFortCustomBodyType Enum=All
   )
   Event BlueprintUpdateAnimation()
  ]}
]
variables [
 Backpack {type=SkeletalMeshComponent}
 BackpackOffsetLocation {type=Vector}
 BackpackOffsetRotation {type=Rotator}
 BackpackOffsetScale {type=Vector}
 Body {type=SkeletalMeshComponentBudgeted}
 isF_Med {type=boolean}
 IsSkydiving {type=boolean}
 OpenChute {type=boolean}
]
